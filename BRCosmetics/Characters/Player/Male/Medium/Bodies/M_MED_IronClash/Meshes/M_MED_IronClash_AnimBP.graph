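# Animation Blueprint for the IronClash medium male body cosmetic, adding secondary motion to its scarves, belt and grenades.
# The parent mesh's pose is copied through a Control Rig and cached, then run through chained spring-based AnimDynamics nodes to the output.
# An unconnected RigidBody node remains in the graph.
# Each update computes Gravity Override with a deprecated skydive-aware wind/gravity function, but no node visibly reads it.
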AnimBlueprint /BRCosmetics/Characters/Player/Male/Medium/Bodies/M_MED_IronClash/Meshes/M_MED_IronClash_AnimBP
target_skeleton=&/M_MED_IronClash_Skeleton
graphs [
 AnimGraph {nodes [
   RigidBody!(ComponentPose=Local To Component(), Alpha, OverrideWorldGravity:Vector="0,0,0", ExternalForce) -> (Pose) {
    bOverrideWorldGravity=true
    alphaBoolBlend {blendOption=Linear}
    simSpaceSettings {
     maxAngularAcceleration=500
     maxAngularVelocity=3
     maxLinearAcceleration=400
     maxLinearVelocity=250
     velocityScaleZ=0
     worldAlpha=1
     externalLinearDragV=<0.1 0.1>
    }
   }
   n1 = Anim Dynamics - Bone: dyn_scarf_01!(ComponentPose=Local To Component(LocalPose=Use cached pose 'PriSim'()), Alpha=1, GravityOverride:Vector="0,0,0", ExternalForce="0,0,0") -> (Pose) {
    angularDampingOverride=1
    bAngularSpring=true
    bGravityOverrideInSimSpace=true
    bOverrideAngularDamping=true
    bUseGravityOverride=true
    gravityScale=0.1
    linearSpringConstant=10
    simulationSpace=RootRelative
    alphaBoolBlend {blendOption=Linear}
    boundBone {boneName=dyn_scarf_01}
    componentAppliedLinearAccClamp=<10000 10000 10000>
    componentLinearAccScale=<1 1 1>
    componentLinearVelScale=<1 1 1>
    physicsBodyDefinitions [{
      boundBone {boneName=dyn_scarf_01}
      boxExtents=<15 15 15>
      constraintSetup {angularLimitsMax=<0 2 20>, angularLimitsMin=<0 -20 -25>, angularTarget=<0 0 0>}
      localJointOffset=<2 0 0>
     }]
    relativeSpaceBone {boneName=dyn_driver_dyn_scarf}
   }
   n2 = Anim Dynamics - Bone: dyn_scarf_02!(ComponentPose=@n1, Alpha=1, GravityOverride:Vector="0,0,0", ExternalForce="0,0,0") -> (Pose) {
    angularDampingOverride=1
    angularSpringConstant=100
    bAngularSpring=true
    bOverrideAngularDamping=true
    bUseGravityOverride=true
    gravityScale=0.1
    linearSpringConstant=10
    simulationSpace=RootRelative
    alphaBoolBlend {blendOption=Linear}
    boundBone {boneName=dyn_scarf_02}
    componentAppliedLinearAccClamp=<10000 10000 10000>
    componentLinearAccScale=<1 1 1>
    componentLinearVelScale=<1 1 1>
    physicsBodyDefinitions [{
      boundBone {boneName=dyn_scarf_02}
      boxExtents=<12 12 12>
      constraintSetup {angularLimitsMax=<0 2 20>, angularLimitsMin=<0 -20 -25>, angularTarget=<0 0 0>}
      localJointOffset=<1 0 0>
     }]
    relativeSpaceBone {boneName=spine_05}
   }
   n3 = Anim Dynamics - Bone: dyn_scarf_03!(ComponentPose=@n2, Alpha=1, GravityOverride:Vector="0,0,0", ExternalForce="0,0,0") -> (Pose) {
    angularDampingOverride=1
    angularSpringConstant=100
    bAngularSpring=true
    bGravityOverrideInSimSpace=true
    bOverrideAngularDamping=true
    bUseGravityOverride=true
    gravityScale=0.1
    linearSpringConstant=10
    simulationSpace=RootRelative
    alphaBoolBlend {blendOption=Linear}
    boundBone {boneName=dyn_scarf_03}
    componentAppliedLinearAccClamp=<10000 10000 10000>
    componentLinearAccScale=<1 1 1>
    componentLinearVelScale=<1 1 1>
    physicsBodyDefinitions [{
      boundBone {boneName=dyn_scarf_03}
      boxExtents=<10 10 10>
      constraintSetup {angularLimitsMax=<0 1 25>, angularLimitsMin=<0 -20 -25>, angularTarget=<0 0 0>}
      localJointOffset=<0.5 0 0>
     }]
    relativeSpaceBone {boneName=spine_05}
   }
   n4 = Anim Dynamics - Bone: Belt!(ComponentPose=@n3, Alpha=0.6, GravityOverride:Vector="0,0,0", ExternalForce="0,0,0") -> (Pose) {
    angularDampingOverride=1
    angularSpringConstant=1000
    bAngularSpring=true
    bOverrideAngularDamping=true
    bUseGravityOverride=true
    gravityScale=0.1
    linearSpringConstant=10
    simulationSpace=RootRelative
    alphaBoolBlend {blendOption=Linear}
    boundBone {boneName=Belt}
    componentAppliedLinearAccClamp=<10000 10000 10000>
    componentLinearAccScale=<1 1 1>
    componentLinearVelScale=<1 1 1>
    physicsBodyDefinitions [{
      boundBone {boneName=Belt}
      boxExtents=<7 7 7>
      constraintSetup {angularLimitsMax=<1 16 1>, angularLimitsMin=<-1 -16 -2>, angularTarget=<0 0 0>}
      localJointOffset=<1 0 0>
     }]
   }
   n5 = Anim Dynamics - Bone: Grenades!(ComponentPose=@n4, Alpha=1, GravityOverride:Vector="0,0,0", ExternalForce="0,0,0") -> (Pose) {
    angularDampingOverride=1
    angularSpringConstant=1000
    bAngularSpring=true
    bOverrideAngularDamping=true
    bUseGravityOverride=true
    gravityScale=0.1
    linearSpringConstant=10
    simulationSpace=RootRelative
    alphaBoolBlend {blendOption=Linear}
    boundBone {boneName=Grenades}
    componentAppliedLinearAccClamp=<3000 3000 3000>
    componentLinearAccScale=<1 1 1>
    componentLinearVelScale=<1 1 1>
    physicsBodyDefinitions [{
      boundBone {boneName=Grenades}
      boxExtents=<5 5 5>
      constraintSetup {twistAxis=AxisY, angularLimitsMax=<2 2 5>, angularLimitsMin=<-2 -2 -5>, angularTarget=<0 0 0>}
      localJointOffset=<0 0 -2.1>
     }]
   }
   n6 = Anim Dynamics - Bone: dyn_back_scarf_01!(ComponentPose=@n5, Alpha=1, GravityOverride:Vector="0,0,0", ExternalForce="0,0,0") -> (Pose) {
    angularDampingOverride=1
    angularSpringConstant=500
    bAngularSpring=true
    bOverrideAngularDamping=true
    bUseGravityOverride=true
    gravityScale=0.1
    linearSpringConstant=10
    simulationSpace=RootRelative
    alphaBoolBlend {blendOption=Linear}
    boundBone {boneName=dyn_back_scarf_01}
    componentAppliedLinearAccClamp=<5000 5000 5000>
    componentLinearAccScale=<1 1 1>
    componentLinearVelScale=<1 1 1>
    physicsBodyDefinitions [{
      boundBone {boneName=dyn_back_scarf_01}
      boxExtents=<11 11 11>
      constraintSetup {angularLimitsMax=<1 20 5>, angularLimitsMin=<-1 -1 -5>, angularTarget=<0 0 0>}
      localJointOffset=<0.31 0 0>
     }]
    relativeSpaceBone {boneName=dyn_driver_back_scarf}
   }
   Output Pose(Result=Component To Local(ComponentPose=@n6)) {layerGroup=None}
   n7 = Control Rig(Source=Copy Pose From Mesh() {bUseAttachedParent=true}) {
    alphaBoolBlend {blendOption=Linear}
    controlRigAssetReference {blueprintRigClass=&/M_MED_IronClash_CtrlRig.M_MED_IronClash_CtrlRig_C}
   }
   Save cached pose 'PriSim'(Pose=@n7) {cacheName=PriSim}
  ]}
 EventGraph {nodes [
   Event BlueprintUpdateAnimation()
   > Sequence()
   > n1 = Gravity Override State Function DEPRECATED(
    CharacterPartAnimInstance=Self-Reference()
    GameTimeInSeconds=GetGameTimeInSeconds()
    GravityOverrideParams=Get Gravity Override Params_0() -> (Gravity Override Params_0:Gravity Override Params Struct Structure)
   )
   > Set Gravity Override(Gravity Override=@n1)
  ]}
]
variables [
 Gravity Override {type=Vector}
 Gravity Override Params {
  type=GravityOverrideParamsStruct
  default {
   skydive_FloatingDown {windAmplitude=<100 -500 400>, windFrequency=<0.5 7 8>}
   skydive_GoingUp {windAmplitude=<100 -500 400>, windFrequency=<0.5 7 8>}
  }
 }
 Gravity Override Params_0 {type=GravityOverrideParamsStruct, default {skydive_Parachuting {gravityOverride=<0 -300 -980>, windAmplitude=<200 -50 400>}}}
]
